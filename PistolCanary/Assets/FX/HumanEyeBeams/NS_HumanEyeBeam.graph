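# Niagara system for a red human eye beam, fired from a character's skeletal mesh toward User.Stop Location.
# RayMain stretches a flickering sprite beam from a sampled mesh point by BeamProgress and BeamWidth, while Flare_Cam swells when the camera looks down the beam.
# Stop-point flares, glows, GPU sparks and sprites, plus a light, respond to User.Firing.
# Older converted-beam and muzzle emitters remain disabled.
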NiagaraSystem /PistolCanary/Assets/FX/HumanEyeBeams/NS_HumanEyeBeam
dependencies {
 usedDataInterfaces [
  &NiagaraDataInterfaceCurve
  &NiagaraDataInterfaceSkeletalMesh
  &NiagaraDataInterfaceArrayFloat2
  &NiagaraDataInterfaceArrayFloat3
  &NiagaraDataInterfaceCamera
  &NiagaraDataInterfaceColorCurve
 ]
 usedDynamicInputs [&FindQuatBetweenTwoVectors]
 usedModules [
  &self:SystemSpawnScript.SystemScriptSource.SystemScriptGraph.NiagaraNodeAssignment_2.SetVariables_#1
  &SystemState
  &self:SystemSpawnScript.SystemScriptSource.SystemScriptGraph.NiagaraNodeAssignment_3.SetVariables_#2
  &self:SystemSpawnScript.SystemScriptSource.SystemScriptGraph.NiagaraNodeAssignment_4.SetVariables_#3
  &self:SystemSpawnScript.SystemScriptSource.SystemScriptGraph.NiagaraNodeAssignment_5.SetVariables_#4
  &self:SystemSpawnScript.SystemScriptSource.SystemScriptGraph.NiagaraNodeAssignment_0.SetVariables_#5
 ]
}
emitters {
 BeamsConverted {
  bEnabled=false
  emitterUpdateScript [
   &EmitterState()
   &SpawnBurst_Instantaneous(Use Spawn Probability=false, UseLoopCountLimit=false)
   &SpawnBurst_Instantaneous(Spawn Group=5, Use Spawn Probability=false, UseLoopCountLimit=false) {module=SpawnBurst_Instantaneous005}
   &SpawnBurst_Instantaneous(Spawn Group=4, Use Spawn Probability=false, UseLoopCountLimit=false) {module=SpawnBurst_Instantaneous004}
   &SpawnBurst_Instantaneous(Spawn Group=3, Use Spawn Probability=false, UseLoopCountLimit=false) {module=SpawnBurst_Instantaneous003}
   &SpawnBurst_Instantaneous(Spawn Group=2, Use Spawn Probability=false, UseLoopCountLimit=false) {module=SpawnBurst_Instantaneous002}
   &SpawnBurst_Instantaneous(Spawn Group=1, Use Spawn Probability=false, UseLoopCountLimit=false) {module=SpawnBurst_Instantaneous001}
   /SetVariables(Emitter.ScaleW:NiagaraFloat=&Waveform("Waveform [1]"=::NewEnumerator12, "[1] Frequency"=20, "[1] Amplitude Min/Max"=<0 1>)) {isSetParameters=true}
   /SetVariables(Emitter.BeamWidthScale:NiagaraFloat=&FloatFromCurve(
     FloatCurve=DataInterface() {propertyValues {Curve {keys [{interpMode=RCIM_Cubic, tangentMode=RCTM_Break, time=0, value=1}, <at 1 0>]}}}
     CurveIndex=$System.TimeForDis
     Scale Curve=$User.BeamWidth
    )) {isSetParameters=true}
  ]
  particleSpawnScript [
   &InitializeParticle(
    Color=&MergeHelper_ColorSwitch(
     0=<rgba 0 0 0 0>
     1=<rgba 0.6 0.03 0.0775 1.5>
     2=<rgba 0 0 0 0>
     3=<rgba 2.5 0.025 0.2312 0.1>
     4=<rgba 0 0 0 0>
     5=<rgba 0 0 0 0>
     Index=&SpawnGroup()
     Num=::NewEnumerator4
    )
    Lifetime=$System.CurrentLoopDuration
    Ribbon Width=&MergeHelper_FloatSwitch(0=150, 1=220, 2=280, 3=700.3717, 4=200, 5=100, Index=&SpawnGroup(), Num=::NewEnumerator4)
    Ribbon Width Mode=::NewEnumerator1
    Sprite Size=<1 1>
    Sprite Size Mode=::NewEnumerator4
    UsePositionOffset=false
    Write Material Random=true
   )
   /SetVariables(Particles.VisibilityTag=&SpawnGroup()) {isSetParameters=true}
  ]
  particleUpdateScript [
   &ParticleState()
   &SkeletalMeshLocation(
    MaskSpawnGroup=false
    OverrideRandomSeed=false
    Skeletal Mesh=DataInterface() {propertyValues {
      DefaultMesh=/BRCosmetics/Characters/Player/Male/Medium/Bodies/M_MED_WealthLamb_Fate/Meshes/M_MED_WealthLamb_Fate
      ExcludeBoneName=root
      OptionalComponentClass=null
      PreviewMesh=&M_MED_WealthLamb_Fate
      SkinningMode=SkinOnTheFly
      SoftSourceActor=null
      SourceMode=Default
      UvSetIndex=0
      WholeMeshLOD=-1
      bExcludeBone=false
      bReadDeformedGeometry=true
      bRequireCurrentFrameData=true
      ComponentTags []
      FilteredBones [head]
      FilteredSockets []
      MeshUserParameter {parameter {name=None, typeDefHandle {registeredTypeIndex=1}, varData []}}
      SamplingRegions []
     }}
   )
   &ScaleSpriteSize(
    Scale Factor=&MakeVector2D(
     X=&MergeHelper_FloatSwitch(
      0=&Multiply_Float(A=$Particles.Initial.RibbonWidth, B=$Emitter.BeamWidthScale)
      1=&Multiply_Float(A=$Particles.Initial.RibbonWidth, B=$Emitter.BeamWidthScale)
      2=&Multiply_Float(A=$Particles.Initial.RibbonWidth, B=$Emitter.BeamWidthScale)
      3=&Multiply_Float(
       A=&Multiply_Float(A=$Particles.Initial.RibbonWidth, B=$Emitter.BeamWidthScale)
       B=/Game/Effects/Niagara/DynamicInputs/Math/FN_Fit(
        Clamp:NiagaraBool=false
        Value:NiagaraFloat=$Emitter.ScaleW
        SourceMin:NiagaraFloat=0
        SourceMax:NiagaraFloat=1
        DestMin:NiagaraFloat=0.25
        DestMax:NiagaraFloat=1
       )
      )
      4=&Multiply_Float(
       A=&Multiply_Float(A=$Particles.Initial.RibbonWidth, B=$Emitter.BeamWidthScale)
       B=&FN_Fit(
        Clamp:NiagaraBool=false
        Value:NiagaraFloat=$Emitter.ScaleW
        SourceMin:NiagaraFloat=0
        SourceMax:NiagaraFloat=1
        DestMin:NiagaraFloat=0.7
        DestMax:NiagaraFloat=1
       )
      )
      5=&Multiply_Float(A=$Particles.Initial.RibbonWidth, B=$Emitter.BeamWidthScale)
      Index=&SpawnGroup()
      Num=::NewEnumerator4
     )
     Y=&DistanceBetweenPositions(
      A=&Lerp_Vector(
       A=Linked() {linkedVariable {name=Engine.Emitter.SimulationPosition, type=NiagaraPosition}}
       B=Linked() {linkedVariable {name=User.Stop Location, type=NiagaraPosition}}
       Alpha=$User.BeamProgress
      )
      B=$Engine.Emitter.SimulationPosition
     )
    )
    Scale Sprite Size Mode=::NewEnumerator2
   )
   &SpriteFacingAndAlignment(
    Sprite Alignment=&NormalizeVector(
     Vector To Normalize=&SubtractVectorFromPosition(Position=$User.Stop Location, Vector=Linked() {linkedVariable {name=Engine.Emitter.SimulationPosition, type=NiagaraPosition}})
     Safe Normalize=true
     Safe Fallback Threshold=0
     Safe Fallback Vector=<0 0 0.001>
    )
    WriteAlignment=true
    WriteFacing=false
   )
   &DynamicMaterialParameters(
    Index 0 Param 1=&MaskFloatBySpawnGroup(Masked Float=2, Passthrough Float=1)
    Index 2 Param 1=&Lerp_Float(
     A=-0.8775
     B=1
     Alpha=&Multiply_Float(
      A=&FloatFromCurve(FloatCurve=DataInterface() {propertyValues {Curve {keys [<at 0 0>,<at 0.1 1>]}}}, CurveIndex=$Engine.System.Age)
      B=1
     )
    )
    Index 3 Param 1=&FN_Fit(
     Clamp:NiagaraBool=false
     Value:NiagaraFloat=$Emitter.ScaleW
     SourceMin:NiagaraFloat=0
     SourceMax:NiagaraFloat=1
     DestMin:NiagaraFloat=0
     DestMax:NiagaraFloat=&FloatFromCurve(
      FloatCurve=DataInterface() {propertyValues {Curve {keys [{interpMode=RCIM_Cubic, tangentMode=RCTM_Break, time=0, value=0}, <at 0.5 1>]}}}
      CurveIndex=$Engine.System.Age
      Scale Curve=&FloatFromCurve(
       FloatCurve=DataInterface() {propertyValues {Curve {keys [{interpMode=RCIM_Cubic, leaveTangent=-3.7341, tangentMode=RCTM_Break, time=0, value=1}, <at 1 -1>]}}}
       CurveIndex=$System.TimeForDis
      )
     )
    )
    Module.Index0Param0Write=true
    Module.Index0Param1Write=true
    Module.Index0Param2Write=false
    Module.Index0Param3Write=false
    Module.Index2Param0Write=true
    Module.Index2Param1Write=true
    Module.Index2Param2Write=true
    Module.Index2Param3Write=true
    Module.Index3Param0Write=true
    Module.Index3Param1Write=true
    Module.Index3Param2Write=true
    Module.Index3Param3Write=true
    Write Parameter Index 2=true
    Write Parameter Index 3=true
   )
   &ScaleColor(
    Scale Alpha=&MergeHelper_FloatSwitch(
     0=1
     1=1
     2=1
     3=&FloatFromCurve(
      FloatCurve=DataInterface() {propertyValues {Curve {keys [
          {interpMode=RCIM_Cubic, tangentMode=RCTM_Break, time=0, value=1}
          {arriveTangent=-3.624, interpMode=RCIM_Cubic, leaveTangent=-0.5922, tangentMode=RCTM_Break, time=0.7, value=0.3}
          {interpMode=RCIM_Cubic, tangentMode=RCTM_Break, time=1, value=0.4}
         ]}}}
      CurveIndex=$Particles.RibbonLinkOrder
     )
     4=1
     5=1
     Index=&SpawnGroup()
     Num=::NewEnumerator4
    )
    ScaleA=true
    ScaleRGB=true
   )
   &ScaleColor(
    Scale Alpha=&FloatFromCurve(
     FloatCurve=DataInterface() {propertyValues {Curve {keys [{interpMode=RCIM_Cubic, leaveTangent=-3.7201, tangentMode=RCTM_Break, time=0, value=1}, <at 1 0>]}}}
     CurveIndex=$System.TimeForDis
    )
    ScaleA=true
    ScaleRGB=true
   ) {module=ScaleColor001}
  ]
  rendererClasses [&NiagaraSpriteRendererProperties]
  renderers [
   {rendererClass=&NiagaraSpriteRendererProperties}
   {rendererClass=&NiagaraSpriteRendererProperties, rendererIndex=1}
   {rendererClass=&NiagaraSpriteRendererProperties, rendererIndex=2}
   {rendererClass=&NiagaraSpriteRendererProperties, rendererIndex=3}
   {rendererClass=&NiagaraSpriteRendererProperties, rendererIndex=4}
   {rendererClass=&NiagaraSpriteRendererProperties, rendererIndex=5}
  ]
 }
 Core_Flare {
  emitterUpdateScript [
   &EmitterState(Inactive Response=::NewEnumerator2, Life Cycle Mode=::NewEnumerator1, Loop Behavior=::NewEnumerator1, Loop Duration=1, UseLoopDelay=false)
   &SpawnBurst_Instantaneous(Spawn Count=2, Use Spawn Probability=false, UseLoopCountLimit=false)
   /SetVariables() {isSetParameters=true}
  ]
  particleSpawnScript [&InitializeParticle(
    Color=<rgba 2 0.0833 0.1239 0.5565>
    Lifetime=5
    Position Mode=::NewEnumerator1
    Sprite Size=&Multiply_Vector2D_ByFloat(
     Float=300
     Vector2D=&SelectVector2DFromArray(
      Direct Array Index=&ReturnExecIndex()
      Direct Array Mode=::NewEnumerator1
      Vector 2D Selection Array=DataInterface() {propertyValues {FloatData [<12 0.2>, <8 2>]}}
     )
    )
    Sprite Size Mode=::NewEnumerator4
    UsePositionOffset=false
    Write Material Random=true
   )]
  particleUpdateScript [
   &ParticleState(Kill Particles When Lifetime Has Elapsed=false)
   &ScaleSpriteSize(
    Scale Sprite Size Mode=::NewEnumerator0
    Uniform Scale Factor=&RandomRangeFloat(Minimum=0.3, Maximum=0.9, Evaluation Type=::NewEnumerator1)
   )
   &ScaleSpriteSize(
    Scale Sprite Size Mode=::NewEnumerator0
    Uniform Scale Factor=&RemapRange(
     Input Value To Remap=/Niagara/DynamicInputs/Camera/DistanceToCamera(Use Owner Position:NiagaraBool=true, Camera Query:NiagaraDataInterfaceCamera=DataInterface())
     Input Min=1000
     Input Max=10000
     Output Min=1
     Output Max=3
    )
   ) {module=ScaleSpriteSize001}
   &ScaleSpriteSize(Scale Sprite Size Mode=::NewEnumerator0, Uniform Scale Factor=$User.Firing) {module=ScaleSpriteSize002}
   /SetVariables(Particles.SpriteRotation001:NiagaraFloat=&RandomRangeFloat(Minimum=0, Maximum=360, Evaluation Type=::NewEnumerator1)) {isSetParameters=true}
   /SetVariables(Particles.Color=&Multiply_LinearColorByFloat(
     Linear Color=$Particles.Initial.Color
     Float=&RemapRange(
      Input Value To Remap=&DistanceToCamera(Use Owner Position:NiagaraBool=true, Camera Query:NiagaraDataInterfaceCamera=DataInterface())
      Input Min=500
      Input Max=2000
      Output Min=0.05
      Output Max=0.3
     )
    )) {enabled=false, isSetParameters=true}
   &CameraOffset(Camera Offset Amount=100)
   /SetVariables(Particles.Position=$User.Stop Location) {isSetParameters=true}
  ]
  rendererClasses [&NiagaraSpriteRendererProperties]
  renderers [{rendererClass=&NiagaraSpriteRendererProperties}]
 }
 Flare_Cam {
  simTarget=GPUComputeSim
  emitterUpdateScript [
   &EmitterState(Life Cycle Mode=::NewEnumerator1, Loop Behavior=::NewEnumerator1, Loop Duration=1, UseLoopDelay=false)
   &SpawnBurst_Instantaneous(Spawn Count=2, Use Spawn Probability=false, UseLoopCountLimit=false)
  ]
  particleSpawnScript [&InitializeParticle(
    Color=<rgba 0.1302 0 0.0028 0.001>
    Lifetime=5
    Position=&SelectVectorFromArray(
     Array Sampling Mode=::NewEnumerator0
     Direct Array Index=?
     Vector Selection Array=DataInterface() {propertyValues {FloatData [<0 0 0>, <1 1 1>]}}
     Random Seed=0
    )
    Position Mode=::NewEnumerator1
    Sprite Size=&Multiply_Vector2D_ByFloat(
     Float=50
     Vector2D=&SelectVector2DFromArray(
      Direct Array Index=&ReturnExecIndex()
      Vector 2D Selection Array=DataInterface() {propertyValues {FloatData [<12 0.1>, <8 1.5>]}}
     )
    )
    Sprite Size Mode=::NewEnumerator4
    UsePositionOffset=false
    Write Material Random=true
   )]
  particleUpdateScript [
   &ParticleState(Kill Particles When Lifetime Has Elapsed=false)
   &ScaleSpriteSize(
    Scale Sprite Size Mode=::NewEnumerator0
    Uniform Scale Factor=&RandomRangeFloat(Minimum=0.8, Maximum=1, Evaluation Type=::NewEnumerator1)
   )
   &ScaleSpriteSize(
    Initial Sprite Size:Vector2f=$Particles.SpriteSize
    Scale Sprite Size Mode=::NewEnumerator0
    Uniform Scale Factor=&RemapRange(
     Input Value To Remap=&DotProduct(
      Vector A=&NormalizeVector(
       Vector To Normalize=&Subtract_Vector(
        A=&ReturnCameraPosition(Coordinate Space:ENiagaraCoordinateSpace=::World, Camera Query:NiagaraDataInterfaceCamera=DataInterface())
        B=Linked() {linkedVariable {name=System.StartPos, type=NiagaraPosition}}
       )
       Safe Normalize=true
       Safe Fallback Threshold=0
       Safe Fallback Vector=<0 0 0.001>
      )
      Vector B=&NormalizeVector(
       Vector To Normalize=&Subtract_Position(A=$System.EndPos, B=$System.StartPos)
       Safe Normalize=true
       Safe Fallback Threshold=0
       Safe Fallback Vector=<0 0 0.001>
      )
     )
     Input Min=0.99
     Input Max=1
     Output Min=0
     Output Max=25
    )
   ) {module=ScaleSpriteSize001}
   /SetVariables(Particles.Color=&Multiply_LinearColorByFloat(
     Linear Color=$Particles.Initial.Color
     Float=&RemapRange(
      Input Value To Remap=&DistanceToCamera(Use Owner Position:NiagaraBool=true, Camera Query:NiagaraDataInterfaceCamera=DataInterface())
      Input Min=500
      Input Max=10000
      Output Min=0.1
      Output Max=1
     )
    )) {isSetParameters=true}
   &CameraOffset(Camera Offset Amount=100) {module=CameraOffset001}
  ]
  rendererClasses [&NiagaraSpriteRendererProperties]
  renderers [{rendererClass=&NiagaraSpriteRendererProperties}]
 }
 Light {
  emitterUpdateScript [
   &EmitterState(Life Cycle Mode=::NewEnumerator1, Loop Behavior=::NewEnumerator1, Loop Duration Mode=::NewEnumerator1, UseLoopDelay=false)
   &SpawnBurst_Instantaneous(Use Spawn Probability=false, UseLoopCountLimit=false)
  ]
  particleSpawnScript [
   &InitializeParticle(
    Color=<rgba 0.526 0 0.0026 1>
    Position Offset=<0 0 2>
    Sprite Size Mode=::NewEnumerator3
    Uniform Sprite Size=5
    UsePositionOffset=true
    Write Material Random=true
   )
   &CameraOffset(Camera Offset Amount=3)
  ]
  particleUpdateScript [
   &ParticleState(Kill Particles When Lifetime Has Elapsed=false)
   &ScaleColor(Scale RGB=&VectorFromFloat(Value=$User.Firing), ScaleA=true, ScaleRGB=true)
  ]
  rendererClasses [&NiagaraLightRendererProperties]
  renderers [{rendererClass=&NiagaraLightRendererProperties, rendererIndex=0}]
 }
 Mudulate {
  emitterUpdateScript [
   &EmitterState(
    Inactive Response=::NewEnumerator1
    Life Cycle Mode=::NewEnumerator1
    Loop Behavior=::NewEnumerator1
    Loop Duration Mode=::NewEnumerator1
    UseLoopDelay=false
   )
   &SpawnBurst_Instantaneous(Spawn Count=2, Use Spawn Probability=false, UseLoopCountLimit=false)
  ]
  particleSpawnScript [&InitializeParticle(
    Color=<rgba 3 0.9688 1.1723 1>
    Lifetime=5
    Sprite Size Mode=::NewEnumerator3
    Uniform Sprite Size=1200
    UsePositionOffset=false
    Write Material Random=true
   )]
  particleUpdateScript [
   &ParticleState(Kill Particles When Lifetime Has Elapsed=false)
   &ScaleSpriteSize(
    Uniform Curve Index=&RandomRangeFloat(Minimum=0.3, Maximum=0.8, Evaluation Type=::NewEnumerator1)
    Uniform Curve Scale=&RemapRange(
     Input Value To Remap=&DistanceToCamera(Use Owner Position:NiagaraBool=true, Camera Query:NiagaraDataInterfaceCamera=DataInterface())
     Input Min=100
     Input Max=1000
     Output Min=0.1
     Output Max=1
    )
   )
   &ScaleSpriteSize(Scale Sprite Size Mode=::NewEnumerator0, Uniform Scale Factor=$User.Firing) {module=ScaleSpriteSize001}
   &DynamicMaterialParameters(
    Index 0 Param 1=&RandomRangeFloat(Minimum=0.3, Maximum=0.6, Evaluation Type=::NewEnumerator1)
    Module.Index0Float0Write=true
    Module.Index0Float1Write=true
    Module.Index0Float2Write=true
    Module.Index0Float3Write=true
   )
   /SetVariables(Particles.Position=$User.Stop Location) {isSetParameters=true}
  ]
  rendererClasses [&NiagaraSpriteRendererProperties]
  renderers [{rendererClass=&NiagaraSpriteRendererProperties}]
 }
 MuzzelParticle {
  bEnabled=false
  emitterUpdateScript [
   &EmitterState()
   &SpawnRate(
    SpawnRate=&FloatFromCurve(
     FloatCurve=DataInterface() {propertyValues {Curve {keys [
         {interpMode=RCIM_Cubic, leaveTangent=0.3372, tangentMode=RCTM_Break, time=3.9, value=0}
         {arriveTangent=1.6707, tangentMode=RCTM_User, time=4, value=1}
         <at 4.1 0>
        ]}}}
     CurveIndex=$Emitter.LoopedAge
     Scale Curve=500
    )
    Use Spawn Probability=false
   ) {enabled=false, module=SpawnRate001}
   &SpawnRate(
    SpawnRate=&Multiply_Float(
     A=100
     B=&FloatFromCurve(
      FloatCurve=DataInterface() {propertyValues {Curve {keys [{interpMode=RCIM_Cubic, leaveTangent=-3.7201, tangentMode=RCTM_Break, time=0, value=1}, <at 1 0>]}}}
      CurveIndex=$System.TimeForDis
     )
    )
    Use Spawn Probability=false
   )
   &SpawnRate(
    SpawnGroup=1
    SpawnRate=&Multiply_Float(
     A=100
     B=&FloatFromCurve(
      FloatCurve=DataInterface() {propertyValues {Curve {keys [{interpMode=RCIM_Cubic, leaveTangent=-3.7201, tangentMode=RCTM_Break, time=0, value=1}, <at 1 0>]}}}
      CurveIndex=$System.TimeForDis
     )
    )
    Use Spawn Probability=false
   ) {module=SpawnRate002}
  ]
  particleSpawnScript [
   &InitializeParticle(
    Color=&MaskLinearColorBySpawnGroup(Masked Linear Color=<rgba 0.05 0 0.0042 5>, Passthrough Linear Color=<rgba 10 0 0.8333 0.5>, Spawn Group Mask=1)
    Lifetime=&RandomRangeFloat(
     Minimum=0.2
     Maximum=&MergeHelper_FloatSwitch(0=0.3, 1=0.4, 2=?, 3=?, Index=&SpawnGroup(), Num=::NewEnumerator0)
    )
    Sprite Size=&Multiply_Vector2D_ByFloat(
     Float=&RandomRangeFloat(Minimum=1, Maximum=2)
     Vector2D=&MergeHelper_Vector2Switch(0=<10 120>, 1=<20 100>, 2=?, 3=?, 4=?, Index=&SpawnGroup(), Num=::NewEnumerator0)
    )
    Sprite Size Mode=::NewEnumerator4
    UsePositionOffset=false
    Write Material Random=true
   )
   &ShapeLocation(MaskSpawnGroup=false, Override Seed=false, Sphere Radius=10, UseOwnerScale=true)
   &AddVelocity(
    Module.Use Falloff Distance=false
    Override Seed=false
    Velocity Mode=::NewEnumerator1
    Velocity Speed=&MaskFloatBySpawnGroup(Masked Float=2200, Passthrough Float=3318.3464, Spawn Group Mask=1)
   )
   &AddVelocity(
    Rotation Coordinate Space=::World
    Velocity=$System.BeamDirection
    Velocity Speed Scale=&MaskFloatBySpawnGroup(Masked Float=1741.5543, Passthrough Float=3265.6787, Spawn Group Mask=1)
   ) {module=AddVelocity001}
   /SetVariables(Particles.VisibilityTag=&SpawnGroup()) {isSetParameters=true}
  ]
  particleUpdateScript [
   &ParticleState()
   &ScaleSpriteSize(
    Scale Factor=&MakeVector2D(
     X=&FloatFromCurve(FloatCurve=DataInterface() {propertyValues {Curve {keys [
          {interpMode=RCIM_Cubic, leaveTangent=11.7691, tangentMode=RCTM_Break, time=0, value=0}
          {arriveTangent=0.0629, interpMode=RCIM_Cubic, leaveTangent=0.0629, tangentMode=RCTM_User, time=0.2, value=1}
          {interpMode=RCIM_Cubic, leaveTangent=0.027, tangentMode=RCTM_Break, time=0.35, value=1}
          {arriveTangent=0.0384, tangentMode=RCTM_User, time=1, value=0}
         ]}}})
     Y=1
    )
    Scale Sprite Size Mode=::NewEnumerator2
   )
   &CurlNoiseForce(Bias Noise=false, Noise Frequency=10, Noise Strength=550.2112, Pan Noise=false) {enabled=false}
   &SolveForcesAndVelocity(Clamp Velocity=false, Limit Acceleration=false)
   &ScaleSpriteSizeBySpeed(Max Scale Factor=<1 2>, Velocity Threshold=3500)
   &FN_Sprite_Attributes(
    Color=&MultiplyLinearColor(
     Linear Color=$Particles.Color
     Scale Factor=/Game/Effects/Niagara/Modules/MergeEmitters/MergeHelper_Vector4Switch(
      0:Vector4f=<0.025 0.025 0.025 1>
      1:Vector4f=<0.5 0.5 0.5 0.3>
      2:Vector4f=?
      3:Vector4f=?
      4:Vector4f=?
      5:Vector4f=?
      6:Vector4f=?
      7:Vector4f=?
      8:Vector4f=?
      9:Vector4f=?
      Index:NiagaraInt32=&SpawnGroup()
      Num:MergeHelper_SwitcherEnum=::NewEnumerator0
     )
    )
    Create Color Attribute=true
    Create SpriteSize Attribute=true
    SpriteSize=&Multiply_Vector2D(
     A=$Particles.SpriteSize
     B=&MaskVector2dBySpawnGroup(Masked Vector2D=<3 1>, Passthrough Vector2D=<5 3>, Spawn Group Mask=1)
    )
    Write to Module Namespace=true
   )
   &FN_Sprite_Attributes(
    Color=&MultiplyLinearColor(Linear Color=<rgba 7 0.35 0.9042 1>, Scale Factor=<0.1 0.1 0.1 0.5>)
    Create Color Attribute=true
    Create SpriteSize Attribute=true
    SpriteSize=&Multiply_Vector2D(A=$Particles.SpriteSize, B=<10 0.5>)
    Write to Module Namespace=true
   ) {enabled=false, module=FN_Sprite_Attributes001}
   &ScaleColor(
    Scale Alpha=&FloatFromCurve(FloatCurve=DataInterface() {propertyValues {Curve {
        defaultValue=1
        keys [
         {arriveTangent=9.0725, interpMode=RCIM_Cubic, leaveTangent=9.0725, tangentMode=RCTM_User, time=0, value=0}
         {arriveTangent=-2.5249, interpMode=RCIM_Cubic, leaveTangent=-2.5249, tangentMode=RCTM_User, time=0.1408, value=0.9847}
         {arriveTangent=-0.0405, interpMode=RCIM_Cubic, leaveTangent=-0.0405, tangentMode=RCTM_User, time=1, value=0}
        ]
       }}})
    ScaleA=true
    ScaleRGB=true
   )
  ]
  rendererClasses [&NiagaraSpriteRendererProperties]
  renderers [
   {rendererClass=&NiagaraSpriteRendererProperties}
   {rendererClass=&NiagaraSpriteRendererProperties, rendererIndex=1}
   {rendererClass=&NiagaraSpriteRendererProperties, rendererIndex=2}
  ]
 }
 MuzzleShape {
  bEnabled=false
  emitterUpdateScript [
   &EmitterState()
   &SpawnBurst_Instantaneous(Use Spawn Probability=false, UseLoopCountLimit=false)
   &SpawnBurst_Instantaneous(Spawn Group=1, Use Spawn Probability=false, UseLoopCountLimit=false) {module=SpawnBurst_Instantaneous001}
  ]
  particleSpawnScript [
   &InitializeParticle(
    Lifetime=$User.Duration
    Mesh Scale Mode=::NewEnumerator3
    Mesh Uniform Scale=26.2282
    Position=<15 0 5>
    Position Mode=::NewEnumerator1
    Sprite Size=&MaskVector2dBySpawnGroup(Masked Vector2D=<200 1700>, Passthrough Vector2D=<813.0714 930.1631>, Spawn Group Mask=1)
    Sprite Size Mode=::NewEnumerator4
    UsePositionOffset=false
    Write Material Random=true
   )
   &CameraOffset(Camera Offset Amount=300)
   /SetVariables(Particles.VisibilityTag=&SpawnGroup()) {isSetParameters=true}
  ]
  particleUpdateScript [
   &ParticleState()
   &ScaleMeshSize(Scale Factor=&VectorFromFloat(Value=&FloatFromCurve(FloatCurve=DataInterface() {propertyValues {Curve {keys [<at 0 0.8>,<at 1 1>]}}}, Scale Curve=10)))
   &DynamicMaterialParameters(
    Index 0 Param 1=0.25
    Index 0 Param 2=-1
    Index 0 Param 3=1
    Index 0 Param 4=1
    Module.Index0Param0Write=true
    Module.Index0Param1Write=true
    Module.Index0Param2Write=true
    Module.Index0Param3Write=true
   )
   &ScaleColor(
    Scale Alpha=&MaskFloatBySpawnGroup(
     Masked Float=1
     Passthrough Float=&FloatFromCurve(FloatCurve=DataInterface() {propertyValues {Curve {keys [<at 0.0999 1.0037 RCIM_Cubic>,<at 1 1>]}}})
     Spawn Group Mask=1
    )
    ScaleA=true
    ScaleRGB=true
   )
   /SetVariables(Particles.Position=&AddVectorToPosition(
     Position=&Lerp_Vector(A=$System.SampleSkeletalMesh.SampledPosition, B=$System.SampleSkeletalMesh001.SampledPosition, Alpha=0)
     Vector=&TransformVector(Vector=&Multiply_VectorByFloat(Vector=<20 0 10>, Float=1), Destination Space=::Simulation)
    )) {enabled=false, isSetParameters=true}
   &SpriteFacingAndAlignment(Alignment Coordinate Space=::World, Sprite Alignment=$System.BeamDirection, WriteAlignment=true, WriteFacing=false)
   &ScaleColor(
    Scale Alpha=&FloatFromCurve(
     FloatCurve=DataInterface() {propertyValues {Curve {keys [{interpMode=RCIM_Cubic, tangentMode=RCTM_Break, time=0, value=0}, <at 0.3 1>]}}}
     CurveIndex=$Engine.System.Age
    )
    ScaleA=true
    ScaleRGB=true
   ) {module=ScaleColor001}
   &ScaleColor(
    Scale Alpha=&FloatFromCurve(
     FloatCurve=DataInterface() {propertyValues {Curve {keys [{interpMode=RCIM_Cubic, leaveTangent=-3.7201, tangentMode=RCTM_Break, time=0, value=1}, <at 1 0>]}}}
     CurveIndex=$System.TimeForDis
    )
    ScaleA=true
    ScaleRGB=true
   ) {module=ScaleColor002}
  ]
  rendererClasses [&NiagaraSpriteRendererProperties]
  renderers [
   {rendererClass=&NiagaraSpriteRendererProperties}
   {rendererClass=&NiagaraSpriteRendererProperties, rendererIndex=1}
  ]
 }
 RayMain {
  emitterUpdateScript [
   &EmitterState(Life Cycle Mode=::NewEnumerator1, Loop Behavior=::NewEnumerator1, Loop Duration=1, UseLoopDelay=false)
   &SpawnBurst_Instantaneous(Use Spawn Probability=false, UseLoopCountLimit=false)
  ]
  particleSpawnScript [&InitializeParticle(Color=<rgba 0.2348 0.0245 0.0245 0.1652>, Lifetime=20, UsePositionOffset=false, Write Material Random=true)]
  particleUpdateScript [
   &ParticleState(Kill Particles When Lifetime Has Elapsed=false)
   &SkeletalMeshLocation(
    Bone Sampling Mode=::NewEnumerator4
    MaskSpawnGroup=false
    Position Sampling=::NewEnumerator2
    Sampled Position Offset=<20 0 15>
    Skeletal Mesh=$System.Skeletal Mesh
   )
   &self:SpriteBeam(
    End:NiagaraPosition=&Lerp_Vector(
     A=Linked() {linkedVariable {name=Particles.SkeletalMeshLocation.SampledPosition, type=NiagaraPosition}}
     B=Linked() {linkedVariable {name=User.Stop Location, type=NiagaraPosition}}
     Alpha=$User.BeamProgress
    )
    Start:NiagaraPosition=$Particles.SkeletalMeshLocation.SampledPosition
    Width:NiagaraFloat=&Multiply_Float(
     A=&FloatFromCurve(
      FloatCurve=DataInterface() {propertyValues {Curve {
         defaultValue=1
         keys [
          {tangentMode=RCTM_User, time=0, value=0}
          {arriveTangent=1.1111, interpMode=RCIM_Cubic, leaveTangent=1.1111, time=0.1, value=0.8}
          <at 0.9 1>
          <at 1 0 RCIM_Cubic>
         ]
        }}}
      Scale Curve=&Multiply_Float(A=$User.BeamWidth, B=15)
     )
     B=&RandomRangeFloat(Minimum=20, Maximum=100, Evaluation Type=::NewEnumerator1)
    )
   ) {module=SpriteBeam}
   &DynamicMaterialParameters(
    Index 0 Param 1=1
    Module.Index0Float0Write=true
    Module.Index0Float1Write=true
    Module.Index0Float2Write=true
    Module.Index0Float3Write=true
   )
  ]
  rendererClasses [&NiagaraSpriteRendererProperties]
  renderers [{rendererClass=&NiagaraSpriteRendererProperties}]
 }
 Sparks {
  simTarget=GPUComputeSim
  emitterUpdateScript [
   &EmitterState(
    Life Cycle Mode=::NewEnumerator1
    Loop Delay=0.15
    Loop Duration=&RandomRangeFloat(Minimum=0.1, Maximum=0.4)
    Recalculate Duration Each Loop=true
    UseLoopDelay=false
   )
   &SpawnBurst_Instantaneous(
    Spawn Count=&RandomRangeInt(Minimum=100, Maximum=800, Evaluation Type=::NewEnumerator1)
    Spawn Probability=$User.Firing
    Use Spawn Probability=true
    UseLoopCountLimit=false
   )
  ]
  particleSpawnScript [
   &InitializeParticle(
    Color=&ColorFromCurve(
     ColorCurve=DataInterface() {propertyValues {
       AlphaCurve {keys [<at 0 1>,<at 1 1>]}
       BlueCurve {keys [<at 0.2162 73.4473>, <at 0.2931 13.9308>, {arriveTangent=-9.6884, leaveTangent=-9.6884, time=0.3895, value=0.7805}, <at 0.9849 0.0469>]}
       GreenCurve {keys [<at 0.2162 178.7382>, <at 0.2931 25.3026>, {arriveTangent=-29.5649, leaveTangent=-29.5649, time=0.3895, value=2.7056}, <at 0.9849 0.3102>]}
       RedCurve {keys [<at 0.2162 1000>, <at 0.2931 132.0939>, {arriveTangent=-91.377, leaveTangent=-91.377, time=0.3895, value=10}, <at 0.9849 1>]}
      }}
     CurveIndex=&RandomRangeFloat(Minimum=0, Maximum=1)
    )
    Lifetime=&FloatFromCurve(
     FloatCurve=DataInterface() {propertyValues {Curve {keys [
         {arriveTangent=-3.4295, interpMode=RCIM_Cubic, leaveTangent=-3.4295, tangentMode=RCTM_User, time=0, value=1}
         {arriveTangent=-0.431, interpMode=RCIM_Cubic, leaveTangent=-0.431, tangentMode=RCTM_User, time=1, value=0}
        ]}}}
     CurveIndex=&RandomRangeFloat(Minimum=0, Maximum=1)
     Scale Curve=3
    )
    Sprite Size Mode=::NewEnumerator3
    Uniform Sprite Size=&FloatFromCurve(
     FloatCurve=DataInterface() {propertyValues {Curve {keys [
         {arriveTangent=-2.2424, interpMode=RCIM_Cubic, leaveTangent=-2.2424, tangentMode=RCTM_User, time=0, value=1.5}
         {arriveTangent=-1, interpMode=RCIM_Cubic, leaveTangent=-1, time=0.2, value=0.725}
         <at 1 0.5 RCIM_Cubic>
        ]}}}
     CurveIndex=&RandomRangeFloat(Minimum=0, Maximum=1)
     Scale Curve=1.5
    )
    UsePositionOffset=false
    Write Material Random=true
   )
   &ShapeLocation(
    Cylinder Height=0
    Cylinder Height Midpoint=0
    Cylinder Radius=&UniformAOrBFloat(A=20, B=100, Distribution Weight=0.412, Evaluation Type=::NewEnumerator0, Recalculate Random Each Loop=true, Override Seed=false)
    MaskSpawnGroup=false
    Override Seed=false
    Rotation Mode=::NewEnumerator2
    Shape Origin=$User.Stop Location
    Shape Primitive=::NewEnumerator1
    Surface Only=true
    Use Endcaps In Surface Only Mode=false
    UseOwnerScale=true
    Yaw / Pitch / Roll=&MakeCustomVectorFromBool(
     Bool=&SetBoolByFloatComparison(
      A=&RandomRangeFloat(Minimum=0, Maximum=1, Random Seed=$Engine.System.TickCount, Fixed Random Seed=true)
      B=1
     )
     True Vector=&RandomVector(Vector Scale=1)
     False Vector=&RandomVector(Vector Scale=1, Randomness Mode=::NewEnumerator1, Random Seed=$Engine.System.TickCount, Fixed Random Seed=true)
    )
   )
   &AddVelocity(
    Module.Use Falloff Distance=false
    Override Seed=false
    Velocity Mode=::NewEnumerator1
    Velocity Speed=&RandomRangeFloat(Minimum=500, Maximum=2000)
   )
  ]
  particleUpdateScript [
   &ParticleState()
   &Drag(
    Drag=&FloatFromCurve(
     FloatCurve=DataInterface() {propertyValues {Curve {defaultValue=1, keys [{tangentMode=RCTM_User, time=0, value=1}, <at 0.1 0.05>, <at 0.3 0.025>, {tangentMode=RCTM_User, time=1, value=0}]}}}
     Scale Curve=20
    )
    Use Linear Drag=true
    Use Rotational Drag=true
   )
   &PointForce(Force Strength=-200, Module.Use Falloff=false, Module.UseLinearForce=false)
   &GravityForce()
   &SolveForcesAndVelocity(Clamp Velocity=false, Limit Acceleration=false)
   &ScaleSpriteSizeBySpeed(Max Scale Factor=<1 25>, Velocity Threshold=2000)
   &ScaleSpriteSize(
    Uniform Curve Scale=&RandomRangeFloat(Minimum=0, Maximum=3, Evaluation Type=::NewEnumerator1)
    Uniform Curve Sprite Scale=DataInterface() {propertyValues {Curve {
       defaultValue=1
       keys [
        {arriveTangent=9.0725, interpMode=RCIM_Cubic, leaveTangent=9.0725, tangentMode=RCTM_User, time=0, value=0}
        {arriveTangent=-2.5249, interpMode=RCIM_Cubic, leaveTangent=-2.5249, tangentMode=RCTM_User, time=0.1408, value=0.9847}
        {arriveTangent=-0.0405, interpMode=RCIM_Cubic, leaveTangent=-0.0405, tangentMode=RCTM_User, time=1, value=0}
       ]
      }}}
   )
   &CameraOffset(Camera Offset Amount=-20)
  ]
  rendererClasses [&NiagaraSpriteRendererProperties]
  renderers [{rendererClass=&NiagaraSpriteRendererProperties}]
 }
 Sparks001 {
  simTarget=GPUComputeSim
  emitterUpdateScript [
   &EmitterState(
    Life Cycle Mode=::NewEnumerator1
    Loop Delay=0.15
    Loop Duration=&RandomRangeFloat(Minimum=0.1, Maximum=0.4)
    Recalculate Duration Each Loop=true
    UseLoopDelay=false
   )
   &SpawnBurst_Instantaneous(
    Spawn Count=&RandomRangeInt(Minimum=10, Maximum=20, Evaluation Type=::NewEnumerator1)
    Spawn Probability=$User.Firing
    Use Spawn Probability=true
    UseLoopCountLimit=false
   )
  ]
  particleSpawnScript [
   &InitializeParticle(
    Color=&ColorFromCurve(
     ColorCurve=DataInterface() {propertyValues {
       AlphaCurve {keys [<at 0 1>,<at 1 1>]}
       BlueCurve {keys [<at 0.2162 73.4473>, <at 0.2931 13.9308>, {arriveTangent=-9.6884, leaveTangent=-9.6884, time=0.3895, value=0.7805}, <at 0.9849 0.0469>]}
       GreenCurve {keys [<at 0.2162 178.7382>, <at 0.2931 25.3026>, {arriveTangent=-29.5649, leaveTangent=-29.5649, time=0.3895, value=2.7056}, <at 0.9849 0.3102>]}
       RedCurve {keys [<at 0.2162 1000>, <at 0.2931 132.0939>, {arriveTangent=-91.377, leaveTangent=-91.377, time=0.3895, value=10}, <at 0.9849 1>]}
      }}
     CurveIndex=&RandomRangeFloat(Minimum=0, Maximum=1)
    )
    Lifetime=&FloatFromCurve(
     FloatCurve=DataInterface() {propertyValues {Curve {keys [
         {arriveTangent=-3.4295, interpMode=RCIM_Cubic, leaveTangent=-3.4295, tangentMode=RCTM_User, time=0, value=1}
         {arriveTangent=-0.431, interpMode=RCIM_Cubic, leaveTangent=-0.431, tangentMode=RCTM_User, time=1, value=0.55}
        ]}}}
     CurveIndex=&RandomRangeFloat(Minimum=0, Maximum=1)
     Scale Curve=3
    )
    Sprite Size Mode=::NewEnumerator3
    Uniform Sprite Size=&FloatFromCurve(
     FloatCurve=DataInterface() {propertyValues {Curve {keys [
         {arriveTangent=-2.2424, interpMode=RCIM_Cubic, leaveTangent=-2.2424, tangentMode=RCTM_User, time=0, value=1.5}
         {arriveTangent=-0.6, interpMode=RCIM_Cubic, leaveTangent=-0.6, time=0.2, value=1.1}
         <at 1 0.9 RCIM_Cubic>
        ]}}}
     CurveIndex=&RandomRangeFloat(Minimum=0, Maximum=1)
     Scale Curve=3
    )
    UsePositionOffset=false
    Write Material Random=true
   )
   &ShapeLocation(
    Cylinder Height=0
    Cylinder Height Midpoint=0
    Cylinder Radius=&UniformAOrBFloat(A=20, B=100, Distribution Weight=0.412, Evaluation Type=::NewEnumerator0, Recalculate Random Each Loop=true, Override Seed=false)
    MaskSpawnGroup=false
    Override Seed=false
    Rotation Mode=::NewEnumerator2
    Shape Origin=$User.Stop Location
    Shape Primitive=::NewEnumerator1
    Surface Only=true
    Use Endcaps In Surface Only Mode=false
    UseOwnerScale=true
    Yaw / Pitch / Roll=&MakeCustomVectorFromBool(
     Bool=&SetBoolByFloatComparison(
      A=&RandomRangeFloat(Minimum=0, Maximum=1, Random Seed=$Engine.System.TickCount, Fixed Random Seed=true)
      B=1
     )
     True Vector=&RandomVector(Vector Scale=1)
     False Vector=&RandomVector(Vector Scale=1, Randomness Mode=::NewEnumerator1, Random Seed=$Engine.System.TickCount, Fixed Random Seed=true)
    )
   )
   &AddVelocity(
    Module.Use Falloff Distance=false
    Override Seed=false
    Velocity Mode=::NewEnumerator1
    Velocity Speed=&RandomRangeFloat(Minimum=200, Maximum=500)
   )
  ]
  particleUpdateScript [
   &ParticleState()
   &Drag(
    Drag=&FloatFromCurve(
     FloatCurve=DataInterface() {propertyValues {Curve {defaultValue=1, keys [{tangentMode=RCTM_User, time=0, value=1}, <at 0.1 0.05>, <at 0.3 0.025>, {tangentMode=RCTM_User, time=1, value=0}]}}}
     Scale Curve=20
    )
    Use Linear Drag=true
    Use Rotational Drag=true
   )
   &GravityForce()
   &Collision(
    Age Colliding Particles=false
    Kill Offscreen Particles=true
    KillOnCollision=false
    Module.Control Rest State Via Surface Normal=false
    Module.EnableMaxCollisionCount=false
    Module.Randomize Collision Normal Vector=false
   )
   &SolveForcesAndVelocity(Clamp Velocity=false, Limit Acceleration=false)
   &ScaleSpriteSize(
    Uniform Curve Scale=&RandomRangeFloat(Minimum=2, Maximum=3, Evaluation Type=::NewEnumerator1)
    Uniform Curve Sprite Scale=DataInterface() {propertyValues {Curve {
       defaultValue=1
       keys [
        {arriveTangent=9.0725, interpMode=RCIM_Cubic, leaveTangent=9.0725, tangentMode=RCTM_User, time=0, value=0}
        {arriveTangent=-2.5249, interpMode=RCIM_Cubic, leaveTangent=-2.5249, tangentMode=RCTM_User, time=0.1408, value=0.9847}
        {arriveTangent=-0.0405, interpMode=RCIM_Cubic, leaveTangent=-0.0405, tangentMode=RCTM_User, time=1, value=0}
       ]
      }}}
   )
  ]
  rendererClasses [&NiagaraSpriteRendererProperties]
  renderers [{rendererClass=&NiagaraSpriteRendererProperties}]
 }
 Sprites {
  emitterUpdateScript [
   &EmitterState()
   &SpawnPerUnit(Module.Use Max Movement Threshold=false, Module.Use Movement Tolerance=false, Spawn Spacing=50, Use Spawn Probability=false) {enabled=false}
   &SpawnRate(
    SpawnRate=&Multiply_Float(
     A=&RemapRange(
      Input Value To Remap=&VectorLength(Vector=$Engine.Owner.Velocity)
      Input Min=100
      Input Max=2000
      Output Min=30
      Output Max=0
     )
     B=$User.Firing
    )
    Use Spawn Probability=false
   )
  ]
  particleSpawnScript [
   &InitializeParticle(
    Color Mode=::NewEnumerator0
    Lifetime=&RandomRangeFloat(Minimum=0.2, Maximum=1)
    UsePositionOffset=true
    Write Material Random=true
   )
   &ShapeLocation(
    MaskSpawnGroup=false
    Offset=&RandomRangeVector(Minimum=<0 0 -7.336>, Maximum=<1 1 43.9325>)
    Offset Mode=::NewEnumerator0
    Override Seed=false
    Shape Origin=$User.Stop Location
    Sphere Radius=&UniformAOrBFloat(A=60, B=200, Evaluation Type=::NewEnumerator0, Recalculate Random Each Loop=true, Override Seed=false)
    UseOwnerScale=true
   )
   /SetVariables(Particles.SpriteFacing=<0 0 1>) {isSetParameters=true}
   /SetVariables(
    Particles.SpriteSize=&Multiply_Vector2D_ByFloat(Float=209.8168, Vector2D=&RandomRangeVector2D(Minimum=<0.6262 0.529>, Maximum=<1 2>))
    Particles.SpriteSize2=&Multiply_Vector2D_ByFloat(Float=328.9493, Vector2D=&RandomRangeVector2D(Minimum=<0.3185 0.7881>, Maximum=<1 1.1688>))
   ) {isSetParameters=true}
   &AddVelocity(
    Module.Use Falloff Distance=false
    Override Seed=false
    Velocity Mode=::NewEnumerator1
    Velocity Speed=&RandomRangeFloat(Minimum=128.5402, Maximum=498.3852)
   )
  ]
  particleUpdateScript [
   &ParticleState()
   &Drag(Drag=3.7465, Use Linear Drag=true, Use Rotational Drag=true)
   &PointForce(Force Strength=-2000, Module.Use Falloff=false, Module.UseLinearForce=false) {enabled=false}
   &SolveForcesAndVelocity(Clamp Velocity=false, Limit Acceleration=false)
   &DynamicMaterialParameters(
    Index 0 Param 1=&Add_Float(
     A=&RandomRangeFloat(Minimum=0, Maximum=100)
     B=&FloatFromCurve(
      FloatCurve=DataInterface() {propertyValues {Curve {
         defaultValue=0
         keys [
          {arriveTangent=4.1785, interpMode=RCIM_Cubic, leaveTangent=4.1785, tangentMode=RCTM_User, time=0, value=0}
          {arriveTangent=1.1179, interpMode=RCIM_Cubic, leaveTangent=1.1179, tangentMode=RCTM_User, time=0.2042, value=0.6228}
          {arriveTangent=0.0224, interpMode=RCIM_Cubic, leaveTangent=0.0224, tangentMode=RCTM_User, time=1, value=1}
         ]
        }}}
      Scale Curve=4
     )
    )
    Index 0 Param 2=&FloatFromCurve(FloatCurve=DataInterface() {propertyValues {Curve {keys [<at 0 0 RCIM_Cubic>,<at 0.2 1.5 RCIM_Cubic>,<at 1 0 RCIM_Cubic>]}}})
    Index 0 Param 3=&FloatFromCurve(FloatCurve=DataInterface() {propertyValues {Curve {keys [<at 0 1 RCIM_Cubic>,<at 1 0.125 RCIM_Cubic>]}}}, Scale Curve=0.1467)
    Module.Index0Param0Write=true
    Module.Index0Param1Write=true
    Module.Index0Param2Write=true
    Module.Index0Param3Write=true
   )
   /SetVariables(
    Particles.DynamicMaterialParameter2:Vector4f=$StackContext.DynamicMaterialParameter
    Particles.DynamicMaterialParameter3:Vector4f=$StackContext.DynamicMaterialParameter
   ) {isSetParameters=true}
   &DynamicMaterialParameters(
    Index 0 Param 1=&Add_Float(
     A=&RandomRangeFloat(Minimum=0, Maximum=100)
     B=&FloatFromCurve(
      FloatCurve=DataInterface() {propertyValues {Curve {
         defaultValue=0
         keys [
          {arriveTangent=4.1785, interpMode=RCIM_Cubic, leaveTangent=4.1785, tangentMode=RCTM_User, time=0, value=0}
          {arriveTangent=1.1179, interpMode=RCIM_Cubic, leaveTangent=1.1179, tangentMode=RCTM_User, time=0.2042, value=0.6228}
          {arriveTangent=0.0224, interpMode=RCIM_Cubic, leaveTangent=0.0224, tangentMode=RCTM_User, time=1, value=1}
         ]
        }}}
      Scale Curve=4
     )
    )
    Index 0 Param 2=&FloatFromCurve(FloatCurve=DataInterface() {propertyValues {Curve {keys [<at 0 0 RCIM_Cubic>,<at 0.2 1.5 RCIM_Cubic>,<at 1 0 RCIM_Cubic>]}}})
    Index 0 Param 3=&FloatFromCurve(FloatCurve=DataInterface() {propertyValues {Curve {keys [<at 0 1 RCIM_Cubic>,<at 1 0.125 RCIM_Cubic>]}}}, Scale Curve=0.1467)
    Module.Index0Param0Write=true
    Module.Index0Param1Write=true
    Module.Index0Param2Write=true
    Module.Index0Param3Write=true
   ) {module=DynamicMaterialParameters001}
   &ScaleColor(
    Scale Alpha=&FloatFromCurve(FloatCurve=DataInterface() {propertyValues {Curve {keys [<at 0 0 RCIM_Cubic>,<at 0.3 0.95 RCIM_Cubic>,<at 1 0 RCIM_Cubic>]}}})
    ScaleA=true
    ScaleRGB=true
   )
   &ScaleSpriteSize(Uniform Curve Sprite Scale=DataInterface() {propertyValues {Curve {keys [{tangentMode=RCTM_User, time=0, value=0.35}, {tangentMode=RCTM_User, time=1, value=1}]}}})
  ]
  rendererClasses [&NiagaraSpriteRendererProperties]
  renderers [
   {rendererClass=&NiagaraSpriteRendererProperties}
   {rendererClass=&NiagaraSpriteRendererProperties, rendererIndex=1}
  ]
 }
 Start_Flare {
  emitterUpdateScript [
   &EmitterState(Life Cycle Mode=::NewEnumerator1, Loop Behavior=::NewEnumerator1, Loop Duration=1, UseLoopDelay=false)
   &SpawnBurst_Instantaneous(Spawn Count=2, Use Spawn Probability=false, UseLoopCountLimit=false)
  ]
  particleSpawnScript [&InitializeParticle(
    Color=<rgba 0.5417 0.02 0.0225 0.1826>
    Lifetime=5
    Position Mode=::NewEnumerator1
    Sprite Size=&Multiply_Vector2D_ByFloat(
     Float=400
     Vector2D=&SelectVector2DFromArray(
      Direct Array Index=&ReturnExecIndex()
      Vector 2D Selection Array=DataInterface() {propertyValues {FloatData [<12 0.4>, <8 1.5>]}}
     )
    )
    Sprite Size Mode=::NewEnumerator4
    UsePositionOffset=false
    Write Material Random=true
   )]
  particleUpdateScript [
   &ParticleState(Kill Particles When Lifetime Has Elapsed=false)
   &ScaleSpriteSize(
    Scale Sprite Size Mode=::NewEnumerator0
    Uniform Scale Factor=&RandomRangeFloat(Minimum=0.5, Maximum=1, Evaluation Type=::NewEnumerator1)
   )
   &ScaleSpriteSize(Scale Sprite Size Mode=::NewEnumerator0, Uniform Scale Factor=$User.BeamWidth) {module=ScaleSpriteSize001}
   &SkeletalMeshLocation(Bone Sampling Mode=::NewEnumerator4, MaskSpawnGroup=false, Sampled Position Offset=<10 30 0>, Skeletal Mesh=$System.Skeletal Mesh) {module=SkeletalMeshLocation001}
   &CameraOffset(Camera Offset Amount=50)
  ]
  rendererClasses [&NiagaraSpriteRendererProperties]
  renderers [{rendererClass=&NiagaraSpriteRendererProperties}]
 }
 Start_Flare001 {
  emitterUpdateScript [
   &EmitterState(Life Cycle Mode=::NewEnumerator1, Loop Behavior=::NewEnumerator1, Loop Duration=1, UseLoopDelay=false)
   &SpawnBurst_Instantaneous(Spawn Count=2, Use Spawn Probability=false, UseLoopCountLimit=false)
  ]
  particleSpawnScript [&InitializeParticle(
    Color=<rgba 2 0.0417 0.0831 0.3>
    Lifetime=0.2
    Sprite Size=&Multiply_Vector2D_ByFloat(
     Float=1900
     Vector2D=&SelectVector2DFromArray(Direct Array Index=$Engine.ExecIndex, Vector 2D Selection Array=DataInterface() {propertyValues {FloatData [<4 0.1>, <1 1>]}})
    )
    Sprite Size Mode=::NewEnumerator4
    UsePositionOffset=false
    Write Material Random=true
   )]
  particleUpdateScript [
   &ParticleState()
   &ScaleSpriteSize(
    Scale Factor=&MakeVector2D(
     X=&FloatFromCurve(FloatCurve=DataInterface() {propertyValues {Curve {defaultValue=0, keys [{tangentMode=RCTM_User, time=0, value=0}, {tangentMode=RCTM_User, time=1, value=1}]}}})
     Y=&FloatFromCurve(FloatCurve=DataInterface() {propertyValues {Curve {defaultValue=1, keys [{tangentMode=RCTM_User, time=0, value=0}, {tangentMode=RCTM_User, time=1, value=1}]}}})
    )
    Scale Sprite Size Mode=::NewEnumerator2
   ) {module=ScaleSpriteSize001}
   &ScaleSpriteSize(
    Scale Sprite Size Mode=::NewEnumerator0
    Uniform Scale Factor=&RandomRangeFloat(Minimum=0.5, Maximum=1, Evaluation Type=::NewEnumerator1)
   )
   &SkeletalMeshLocation(Bone Sampling Mode=::NewEnumerator4, MaskSpawnGroup=false, Sampled Position Offset=<10 30 0>, Skeletal Mesh=$System.Skeletal Mesh) {module=SkeletalMeshLocation001}
  ]
  rendererClasses [&NiagaraSpriteRendererProperties]
  renderers [{rendererClass=&NiagaraSpriteRendererProperties}]
 }
}
summary {
 systemName=NS_HumanEyeBeam
 emitters [
  {bEnabled=false, emitterName=MuzzelParticle, rendererClasses [&NiagaraSpriteRendererProperties]}
  {bEnabled=false, emitterName=MuzzleShape, rendererClasses [&NiagaraSpriteRendererProperties]}
  {bEnabled=false, emitterName=BeamsConverted, rendererClasses [&NiagaraSpriteRendererProperties]}
  {emitterName=Start_Flare, rendererClasses [&NiagaraSpriteRendererProperties]}
  {emitterName=Start_Flare001, rendererClasses [&NiagaraSpriteRendererProperties]}
  {emitterName=RayMain, rendererClasses [&NiagaraSpriteRendererProperties]}
  {emitterName=Flare_Cam, simTarget=GPUComputeSim, rendererClasses [&NiagaraSpriteRendererProperties]}
  {emitterName=Light, rendererClasses [&NiagaraLightRendererProperties]}
  {emitterName=Sprites, rendererClasses [&NiagaraSpriteRendererProperties]}
  {emitterName=Mudulate, rendererClasses [&NiagaraSpriteRendererProperties]}
  {emitterName=Core_Flare, rendererClasses [&NiagaraSpriteRendererProperties]}
  {emitterName=Sparks001, simTarget=GPUComputeSim, rendererClasses [&NiagaraSpriteRendererProperties]}
  {emitterName=Sparks, simTarget=GPUComputeSim, rendererClasses [&NiagaraSpriteRendererProperties]}
 ]
 userVariables [
  User.BeamEnd_0 {type=NiagaraPosition, description="", defaultValue {}}
  User.BeamProgress {type=NiagaraFloat, defaultValue=1, description=""}
  User.BeamStart_0_0 {type=NiagaraPosition, description="", defaultValue {}}
  User.BeamWidth {type=NiagaraFloat, defaultValue=1, description=""}
  User.bIsFrontEnd {type=NiagaraBool}
  User.bIsFrontEndPreview {type=NiagaraBool}
  User.Color {type=LinearColor, description="", defaultValue {struct=LinearColor, value=<rgba 0 0.161 0.5 1>}}
  User.Firing {type=NiagaraFloat, defaultValue=1, description=""}
  User.Stop Location {type=NiagaraPosition, description="", defaultValue {}}
  User.WidthMult {type=NiagaraFloat, defaultValue=1, description=""}
 ]
}
